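# Translucent, unlit, two-sided halo effect material for the Tower Sentinel character, used on Niagara mesh particles.
# Panning, distorted noise and flame textures drive a curve-atlas color gradient, plus a tinted smoke highlight, scaled by particle color and time-of-day tints.
# Opacity combines flame shape, noisy bottom mask, inverse Fresnel and an Appear dynamic-parameter sphere-mask reveal, with depth fade skipped on mobile.
# Noise along vertex normals computes a displacement offset.
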Material /Game/Characters/Player/Female/Medium/Bodies/F_MED_Tower_Sentinel/FX/M_F_TowerSentinel_halo_color.M_F_TowerSentinel_Halo_Color
graph {
 properties {
  bUsedWithNiagaraMeshParticles=true
  blendMode=BLEND_Translucent
  pixelDepthOffsetMode=PDOM_Legacy
  shadingModel=MSM_Unlit
  twoSided=true
  usedShadingModels=Unlit
 }
 nodes [
  Constant2Vector() {g=0.8}
  Frac(None=Time())
  n1 = VectorParameter() {group=Color_time_of_day, parameterName=Morning_Color, defaultValue=<rgba 1 1 1 1>}
  n2 = VectorParameter() {group=Color_time_of_day, parameterName=Day_Color, defaultValue=<rgba 1 1 1 1>}
  n3 = VectorParameter() {group=Color_time_of_day, parameterName=Evening_Color, defaultValue=<rgba 1 1 1 1>}
  n4 = VectorParameter() {group=Color_time_of_day, parameterName=Night_Color, defaultValue=<rgba 1 1 1 1>}
  n5 = ParticleColor()
  n6 = Panner(Coordinate=TextureCoordinate() {uTiling=0.5, vTiling=0.2}) {speedY=0.1}
  n7 = VectorParameter() {parameterName=Noise_base_UV, defaultValue=<rgba 1 0.3 0 0>}
  n8 = Panner(Coordinate=Multiply(A=TextureCoordinate(), B=ComponentMask(None=@n7.RGB) {g=true, r=true})) {speedY=0.35}
  n9 = Multiply(A=ComponentMask(None=TextureSample(UVs=@n8) {texture=&T_NoisePack_01}.RGB) {g=true, r=true}) {constB=0.2}
  n10 = LinearInterpolate(Alpha=TextureSampleParameter2D(UVs=Add(A=@n6, B=@n9)) {parameterName=HighLight_Texture, texture=&T_Inky_Smoke_Tile}.R) {constA=-0.2}
  n11 = Constant() {r=1}
  n12 = Constant() {r=3}
  n13 = OneMinus(None=ComponentMask(None=TextureCoordinate()) {g=true})
  n14 = Multiply(A=@n10, B=Power(Base=@n10, Exp=LinearInterpolate(A=@n11, B=@n12, Alpha=@n13) {constA=0.6, constB=4}))
  n15 = VectorParameter() {parameterName=HighLight_color, defaultValue=<rgba 0 6.3495 10 1>}
  n16 = ComponentMask(None=TextureCoordinate()) {g=true}
  n17 = Power(Base=Multiply(A=Multiply(A=@n14, B=@n15.RGB), B=Power(Base=@n16) {constExponent=5}))
  n18 = TextureCoordinate()
  n19 = AppendVector(A=Constant(), B=VertexColor().R)
  n20 = VectorParameter() {parameterName=Default_texture_UV, defaultValue=<rgba 1 0.75 6 0.5>}
  n21 = Panner(Coordinate=Multiply(A=Add(A=@n18, B=@n19), B=ComponentMask(None=@n20.RGB) {g=true, r=true})) {speedX=0.1, speedY=0.2}
  n22 = Multiply(
   A=TextureCoordinate() {uTiling=0.5, vTiling=0.2}
   B=ComponentMask(None=VectorParameter() {parameterName=Disturb_texture_UV, defaultValue=<rgba 2 0.75 6 0.5>}.RGB) {g=true, r=true}
  )
  n23 = Constant()
  n24 = VertexColor()
  n25 = Multiply(
   A=Add(A=AppendVector(A=@n23, B=@n24.R), B=TextureCoordinate())
   B=ComponentMask(None=VectorParameter() {parameterName=Noise_base_UV, defaultValue=<rgba 1 0.3 0 0>}.RGB) {g=true, r=true}
  )
  n26 = ComponentMask(None=TextureSample(UVs=Panner(Coordinate=@n25) {speedX=0.03, speedY=0.2}) {samplerType=SAMPLERTYPE_LinearColor, texture=&T_NoisePack_LinearColor}.RGB) {g=true, r=true}
  n27 = TextureSampleParameter2D(UVs=Add(A=Panner(Coordinate=@n22) {speedX=0.03, speedY=0.2}, B=Multiply(A=@n26) {constB=0.1})) {parameterName=Disturb_texture, samplerType=SAMPLERTYPE_Masks, texture=/Game/Global/Textures/Default/Noise/cellularNoise}
  n28 = Multiply(A=Constant() {r=0.5}, B=ComponentMask(None=@n27.RGB) {g=true, r=true})
  n29 = Subtract(A=TextureSample(UVs=Add(A=@n21, B=@n28)) {samplerType=SAMPLERTYPE_LinearColor, texture=&T_NoisePack_LinearColor}.R) {constB=0.5}
  n30 = ComponentMask(None=TextureCoordinate()) {g=true}
  n31 = OneMinus(None=Abs(None=@n27.R))
  n32 = Add(A=AppendVector(A=Constant(), B=VertexColor().R), B=TextureCoordinate())
  n33 = ComponentMask(None=VectorParameter() {parameterName=Flame_Texture_UV, defaultValue=<rgba 1.5 0.5 0 0>}.RGB) {g=true, r=true}
  n34 = TextureSample(UVs=Panner(Coordinate=@n25) {speedY=0.35}) {texture=&T_NoisePack_01}
  n35 = Add(A=Panner(Coordinate=Multiply(A=@n32, B=@n33)) {speedY=0.55}, B=Multiply(A=ComponentMask(None=@n34.RGB) {g=true, r=true}) {constB=0.2})
  n36 = ScalarParameter() {parameterName=Streaks_LerpA}
  n37 = ScalarParameter() {defaultValue=3, parameterName=Streaks_LerpB}
  n38 = OneMinus(None=ComponentMask(None=TextureCoordinate()) {g=true})
  n39 = TextureCoordinate()
  n40 = Multiply(
   A=Power(
    Base=TextureSampleParameter2D(UVs=@n35) {parameterName=Flame_Texture, texture=/Game/Weapons/WeaponSkins/Wraps/Textures/PlasmaClouds/T_Wrap_PlasmaClouds__M}.G
    Exp=LinearInterpolate(A=@n36, B=@n37, Alpha=@n38) {constA=0.6, constB=4}
   )
   B=Power(Base=ComponentMask(None=@n39) {g=true}, Exp=ScalarParameter() {defaultValue=2, parameterName=Central Contast}) {constExponent=10}
  )
  n41 = Add(A=Multiply(A=@n31, B=@n31), B=@n40)
  n42 = CurveAtlasRowParameter(CurveTime=Multiply(A=Add(A=@n29, B=@n30), B=@n41)) {
   atlas=/Game/Athena/LinearColorCurves/VFX_CurveAtlas
   curve=&/Tower_Sentinel_Halo_Color
   defaultValue=12
  }
  n43 = LinearInterpolate(A=@n42, Alpha=ScalarParameter() {parameterName=Color_lerp})
  n44 = LinearInterpolate(
   A=Add(A=Multiply(A=@n17) {constB=2}, B=Multiply(A=@n43, B=@n41))
   B=Constant() {r=1}
   Alpha=ScalarParameter() {parameterName=Use_Simple_flame_panner}
  )
  MaterialFunctionCall(Emissive=Multiply(
    A=Multiply(A=@n5.RGB, B=@n44)
    B=MaterialFunctionCall(Color_Morning=@n1.RGB, Color_Day=@n2.RGB, Color_Evening=@n3.RGB, Color_Night=@n4.RGB) {materialFunction=&Color_Time_of_Day}.Result
   )) {materialFunction=&MF_FX_EmissiveAdjustments}
  n45 = TextureCoordinate() {uTiling=6, vTiling=0.3}
  n46 = ComponentMask(None=TextureSample(UVs=Panner(Coordinate=@n45) {speedY=0.3}) {texture=&T_NoisePack_01}.RGB) {g=true, r=true}
  n47 = DynamicParameter() {paramNames [Appear,Param2,Param3,Param4]}
  n48 = SphereMask(
   A=Add(A=TextureCoordinate() {uTiling=0}, B=Multiply(A=@n46) {constB=0.3})
   B=AppendVector(A=Constant(), B=LinearInterpolate(Alpha=@n47.Appear) {constA=2, constB=0.8})
  ) {attenuationRadius=0.5, hardnessPercent=0}
  n49 = TextureCoordinate()
  n50 = Multiply(
   A=@n41
   B=Power(Base=ComponentMask(None=@n49) {g=true}, Exp=ScalarParameter() {defaultValue=2, parameterName=Central Contast}) {constExponent=10}
  )
  n51 = MaterialFunctionCall(Float2=TextureCoordinate()) {materialFunction=&BreakOutFloat2Components}
  n52 = Multiply(
   A=TextureCoordinate()
   B=ComponentMask(None=VectorParameter() {parameterName=Bottom_mask_noise_UV, defaultValue=<rgba 1.5 0.5 0 0>}.RGB) {g=true, r=true}
  )
  n53 = Multiply(A=TextureSample(UVs=Panner(Coordinate=@n52) {speedY=0.3}) {texture=&T_NoisePack_01}.R) {constB=0.15}
  n54 = OneMinus(None=Multiply(A=Add(A=@n51.G, B=@n53)))
  n55 = Saturate(None=Multiply(A=Add(A=@n54) {constB=0}, B=ScalarParameter() {defaultValue=15, parameterName=bottom}) {constB=15})
  n56 = Multiply(
   A=Saturate(None=Multiply(A=@n48) {constB=2})
   B=LinearInterpolate(A=Multiply(A=@n50, B=@n55), B=Multiply(A=@n55, B=@n40), Alpha=ScalarParameter() {parameterName=Use_Simple_flame_panner})
  )
  n57 = ComponentMask(None=TextureCoordinate()) {r=true}
  n58 = Multiply(A=Power(Base=Sine(None=@n57) {period=2})) {constB=5}
  n59 = Fresnel() {baseReflectFraction=0}
  n60 = Multiply(A=Multiply(A=Multiply(A=@n5.A, B=@n56), B=Saturate(None=@n58)), B=Power(Base=OneMinus(None=@n59)) {constExponent=1})
  n61 = DepthFade(Opacity=@n60) {fadeDistanceDefault=1}
  Multiply(
   A=ShadingPathSwitch(Default=@n61, Deferred=@n61, Forward=@n61, Mobile=@n60)
   B=MaterialFunctionCall(Value_Morning=@n1.A, Value_Day=@n2.A, Value_Evening=@n3.A, Value_Night=@n4.A) {materialFunction=&Scalar_Time_of_Day}.Result
  )
  n62 = TextureSample(UVs=Panner(Coordinate=TextureCoordinate() {uTiling=3, vTiling=0.5}) {speedY=0.5}) {samplerType=SAMPLERTYPE_Grayscale, texture=&LowResBlurredNoise}
  n63 = ComponentMask(None=TextureCoordinate() {uTiling=3}) {g=true}
  n64 = Power(Base=OneMinus(None=LinearInterpolate(Alpha=@n63))) {constExponent=1}
  Multiply(A=Multiply(A=Multiply(A=@n62.R, B=@n64), B=VertexNormalWS())) {constB=10}
  n65 = ComponentMask(None=TextureCoordinate()) {g=true}
  Saturate(None=Add(A=LinearInterpolate(Alpha=@n65) {constA=-1}, B=LinearInterpolate() {constA=-1.5, constB=1.5}))
  ScalarParameter() {defaultValue=1, parameterName=Glider_appear}
 ]
}
