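# Material function producing the animated, flame-like chest emissive for the Tower Sentinel female character body.
# Panning, noise-distorted flame and cellular layers, shaped by V-coordinate gradients from the UV input, form a mask that drives a curve-atlas color lookup.
# A cyan smoke highlight is added, and the sum is tinted by time of day.
# Shading path switches drop the highlight, cellular detail and lookup noise on Forward and Mobile.
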MaterialFunction /Game/Characters/Player/Female/Medium/Bodies/F_MED_Tower_Sentinel/Materials/MF_TowerSentinelChestEmissive
graph {nodes [
  n1 = NamedRerouteUsage()
  n2 = Multiply(A=NamedRerouteUsage(), B=ComponentMask(None=VectorParameter() {group=TowerSentinel, parameterName=Noise_base_UV}.RGB) {g=true, r=true})
  n3 = ComponentMask(None=TextureSample(UVs=Panner(Coordinate=@n2) {speedY=0.35}) {samplerSource=SSM_Wrap_WorldGroupSettings, texture=&T_NoisePack_01}.RGB) {g=true, r=true}
  n4 = TextureSampleParameter2D(UVs=Add(A=Panner(Coordinate=@n1) {speedY=0.1}, B=Multiply(A=@n3) {constB=0.2})) {
   group=TowerSentinel
   parameterName=HighLight_Texture
   samplerSource=SSM_Wrap_WorldGroupSettings
   texture=&T_Inky_Smoke_Tile
  }
  n5 = LinearInterpolate(Alpha=@n4.R) {constA=-0.2}
  n6 = ComponentMask(None=NamedRerouteUsage()) {g=true}
  n7 = Power(Base=@n5, Exp=LinearInterpolate(A=Constant() {r=1}, B=Constant() {r=3}, Alpha=OneMinus(None=@n6)) {constA=0.6, constB=4})
  n8 = NamedRerouteUsage()
  n9 = Multiply(
   A=Multiply(A=Multiply(A=@n5, B=@n7), B=VectorParameter() {group=TowerSentinel, parameterName=HighLight_color, defaultValue=<rgba 0 6.3495 10 1>}.RGB)
   B=Power(Base=ComponentMask(None=@n8) {g=true}) {constExponent=5}
  )
  n10 = Constant()
  n11 = ShadingPathSwitch(Default=Multiply(A=Power(Base=@n9)) {constB=2}, Forward=@n10, Mobile=@n10)
  n12 = Multiply(
   A=NamedRerouteUsage()
   B=ComponentMask(None=VectorParameter() {group=TowerSentinel, parameterName=Default_texture_UV, defaultValue=<rgba 1 0.75 6 0.5>}.RGB) {g=true, r=true}
  )
  n13 = Constant() {r=0.5}
  n14 = Multiply(
   A=NamedRerouteUsage()
   B=ComponentMask(None=VectorParameter() {group=TowerSentinel, parameterName=Disturb_texture_UV, defaultValue=<rgba 2 0.75 6 0.5>}.RGB) {g=true, r=true}
  )
  n15 = Multiply(
   A=NamedRerouteUsage()
   B=ComponentMask(None=VectorParameter() {group=TowerSentinel, parameterName=Noise_base_UV, defaultValue=<rgba 1 0.3 0 0>}.RGB) {g=true, r=true}
  )
  n16 = TextureSample(UVs=Panner(Coordinate=@n15) {speedX=0.03, speedY=0.2}) {
   samplerSource=SSM_Wrap_WorldGroupSettings
   samplerType=SAMPLERTYPE_LinearColor
   texture=&T_NoisePack_LinearColor
  }
  n17 = ComponentMask(None=@n16.RGB) {g=true, r=true}
  n18 = TextureSampleParameter2D(UVs=Add(A=Panner(Coordinate=@n14) {speedX=0.03, speedY=0.2}, B=Multiply(A=@n17) {constB=0.1})) {
   group=TowerSentinel
   parameterName=Disturb_texture
   samplerSource=SSM_Wrap_WorldGroupSettings
   samplerType=SAMPLERTYPE_Masks
   texture=/Game/Global/Textures/Default/Noise/cellularNoise
  }
  n19 = ComponentMask(None=@n18.RGB) {g=true, r=true}
  n20 = TextureSample(UVs=Add(A=Panner(Coordinate=@n12) {speedX=0.1, speedY=0.2}, B=Multiply(A=@n13, B=@n19))) {
   samplerSource=SSM_Wrap_WorldGroupSettings
   samplerType=SAMPLERTYPE_LinearColor
   texture=&T_NoisePack_LinearColor
  }
  n21 = Add(A=Subtract(A=@n20.R) {constB=0.5}, B=ComponentMask(None=NamedRerouteUsage()) {g=true})
  n22 = Constant() {r=1}
  n23 = OneMinus(None=Abs(None=@n18.R))
  n24 = Constant()
  n25 = Multiply(
   A=NamedRerouteUsage()
   B=ComponentMask(None=VectorParameter() {group=TowerSentinel, parameterName=Flame_Texture_UV, defaultValue=<rgba 1.5 0.5 0 0>}.RGB) {g=true, r=true}
  )
  n26 = ComponentMask(None=TextureSample(UVs=Panner(Coordinate=@n15) {speedY=0.35}) {samplerSource=SSM_Wrap_WorldGroupSettings, texture=&T_NoisePack_01}.RGB) {g=true, r=true}
  n27 = TextureSampleParameter2D(UVs=Add(A=Panner(Coordinate=@n25) {speedY=0.55}, B=Multiply(A=@n26) {constB=0.2})) {
   group=TowerSentinel
   parameterName=Flame_Texture
   samplerSource=SSM_Wrap_WorldGroupSettings
   texture=/Game/Weapons/WeaponSkins/Wraps/Textures/PlasmaClouds/T_Wrap_PlasmaClouds__M
  }
  n28 = NamedRerouteUsage()
  n29 = LinearInterpolate(
   A=ScalarParameter() {group=TowerSentinel, parameterName=Streaks_LerpA}
   B=ScalarParameter() {defaultValue=3, group=TowerSentinel, parameterName=Streaks_LerpB}
   Alpha=OneMinus(None=ComponentMask(None=@n28) {g=true})
  ) {constA=0.6, constB=4}
  n30 = ComponentMask(None=NamedRerouteUsage()) {g=true}
  n31 = ScalarParameter() {group=TowerSentinel, parameterName=Central Contast}
  n32 = Add(
   A=ShadingPathSwitch(Default=Multiply(A=@n23, B=@n23), Forward=@n24, Mobile=@n24)
   B=Multiply(A=Power(Base=@n27.G, Exp=@n29), B=Power(Base=@n30, Exp=@n31) {constExponent=10})
  )
  n33 = CurveAtlasRowParameter(CurveTime=Multiply(A=ShadingPathSwitch(Default=@n21, Forward=@n22, Mobile=@n22), B=@n32) {constA=1}) {
   atlas=/Game/Athena/LinearColorCurves/VFX_CurveAtlas
   curve=/Game/Characters/Player/Female/Medium/Bodies/F_MED_Tower_Sentinel/FX/Tower_Sentinel_Halo_Color
   defaultValue=12
   group=TowerSentinel
  }
  n34 = Multiply(A=LinearInterpolate(A=@n33, Alpha=ScalarParameter() {group=TowerSentinel, parameterName=Color_lerp}), B=@n32)
  n35 = VectorParameter() {group=TowerSentinel_Color_time_of_day, parameterName=Morning_Color, defaultValue=<rgba 1 1 1 1>}
  n36 = VectorParameter() {group=TowerSentinel_Color_time_of_day, parameterName=Day_Color, defaultValue=<rgba 1 1 1 1>}
  n37 = VectorParameter() {group=TowerSentinel_Color_time_of_day, parameterName=Evening_Color, defaultValue=<rgba 1 1 1 1>}
  n38 = VectorParameter() {group=TowerSentinel_Color_time_of_day, parameterName=Night_Color, defaultValue=<rgba 1 1 1 1>}
  FunctionOutput(None=Multiply(
    A=Add(A=@n11, B=@n34)
    B=MaterialFunctionCall(Color_Morning=@n35.RGB, Color_Day=@n36.RGB, Color_Evening=@n37.RGB, Color_Night=@n38.RGB) {materialFunction=&Color_Time_of_Day}.Result
   ))
  NamedRerouteDeclaration(None=FunctionInput(Preview=TextureCoordinate()) {bUsePreviewValueAsDefault=true, inputType=FunctionInput_Vector2}) {name=UV_In, nodeColor=<rgba 0.4062 1 0 1>}
 ]}
