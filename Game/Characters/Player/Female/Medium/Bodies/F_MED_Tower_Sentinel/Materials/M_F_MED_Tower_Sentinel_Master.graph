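# Master skinned material for the Tower Sentinel female player body.
# Shared character PreFX/PostFX functions wrap switchable skin effects like silk, clear coat, gem, thin film and metal LUT.
# It adds optional vertex-color sock hiding and a masked Fresnel anodized tint with sun glint, zeroed on Forward and Mobile.
# An optional masked chest emissive uses radial UVs from pre-skinned positions, cross-fading two half-offset samples, likely hiding the seam.
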Material /Game/Characters/Player/Female/Medium/Bodies/F_MED_Tower_Sentinel/Materials/M_F_MED_Tower_Sentinel_Master
graph {
 properties {
  bUseMaterialAttributes=true
  bUsedWithMeshDeformer=true
  bUsedWithMorphTargets=true
  bUsedWithSkeletalMesh=true
  ditherOpacityMask=true
  pixelDepthOffsetMode=PDOM_Legacy
  shadingModel=MSM_FromMaterialExpression
  subsurfaceProfile=&SS_HeroSkin_02
  usedShadingModels=Default Lit | Clear Coat | Subsurface Profile
 }
 nodes [
  n1 = Subtract(
   A=VectorParameter() {group=TowerSentinel_RadialUV, parameterName=RadialUVOffset2, defaultValue=<rgba 0 0 130 0>}.RGB
   B=VertexInterpolator(VS=PreSkinnedPosition()).PS
  )
  n2 = MaterialFunctionCall(Float2=MaterialFunctionCall(
    UVs=ComponentMask(None=@n1) {b=true, r=true}
    Texture Scale=VectorParameter() {group=TowerSentinel_RadialUV, parameterName=RadialUVScale, defaultValue=<rgba 1 1 1 0>}.RGB
   ) {materialFunction=&ScaleUVsByCenter}.UVs) {materialFunction=&BreakOutFloat2Components}
  n3 = MaterialFunctionCall(X=@n2.R, Y=ComponentMask(None=@n1) {g=true}, Z=@n2.R) {materialFunction=&MakeFloat3}
  n4 = Divide(A=@n3.Result, B=PreSkinnedLocalBounds().Extents)
  n5 = Divide(A=Arctangent2Fast(Y=ComponentMask(None=@n4) {b=true}, X=ComponentMask(None=@n4) {r=true}), B=Constant() {r=3.1416})
  n6 = Distance(A=ComponentMask(None=@n3.Result) {b=true, r=true}, B=Constant())
  n7 = ObjectRadius()
  n8 = AppendVector(A=Multiply(A=Add(A=@n5) {constB=-0.5}) {constB=0.5}, B=OneMinus(None=Divide(A=@n6, B=@n7)))
  n9 = ComponentMask(None=VectorParameter() {group=TowerSentinel_RadialUV, parameterName=UV_Mul, defaultValue=<rgba 1 1 0 0>}.RGB) {g=true, r=true}
  NamedRerouteDeclaration(None=Frac(None=Multiply(A=@n8, B=@n9))) {name=RadialUV, nodeColor=<rgba 1 0.2812 0 1>}
  n10 = MaterialFunctionCall() {materialFunction=&BaseFN_CharacterFunction_PreFX}
  n11 = MaterialFunctionCall!(In=TextureObjectParameter() {group=Skin FX, parameterName=SkinFX_Mask}) {materialFunction=/Game/Weapons/WeaponSkins/MaterialFunction/MF_GlitchSubfunction}
  n12 = StaticSwitchParameter(True=@n11.Glitched, False=@n11.Glitched) {group=Skin FX, parameterName=UseGlitch}
  n13 = StaticSwitchParameter(
   True=MaterialFunctionCall!(Alpha=ChannelMaskParameter(None=@n12) {group=Skin FX, parameterName=Silk Channel}.R, Material=@n10.Result) {materialFunction=/Game/Weapons/WeaponSkins/MaterialFunction/MF_Silk}.Result
   False=@n10.Result
  ) {group=Skin FX, parameterName=Use Silk}
  n14 = StaticSwitchParameter(
   True=MaterialFunctionCall!(Alpha=StaticComponentMaskParameter(None=@n12) {defaultR=true, group=Skin FX, parameterName=Clear Coat Channel}, In=@n13) {materialFunction=/Game/Weapons/WeaponSkins/MaterialFunction/Clearcoat}.Result
   False=@n13
  ) {group=Skin FX, parameterName=Use Clear Coat}
  n15 = StaticSwitchParameter(
   True=MaterialFunctionCall!(
    MASK=StaticComponentMaskParameter(None=@n12) {defaultR=true, group=Skin FX, parameterName=Generic Gem Effect Channel}
    Color Change Sequence=StaticComponentMaskParameter(None=@n12) {defaultB=true, group=Skin FX, parameterName=Generic Gem Color Change Channel}
    Material=@n14
   ) {materialFunction=/Game/Weapons/WeaponSkins/MaterialFunction/GenericGem}.Result
   False=@n14
  ) {group=Skin FX, parameterName=Use Generic Gem}
  n16 = StaticSwitchParameter(
   True=MaterialFunctionCall!(Alpha=StaticComponentMaskParameter(None=@n12) {defaultR=true, group=Skin FX, parameterName=ThinFilm_Channel}, Material Attributes=@n15) {materialFunction=/Game/Weapons/WeaponSkins/MaterialFunction/ThinFilmFunction}.Result
   False=@n15
  ) {group=Skin FX, parameterName=Use Thin Film}
  n17 = MaterialFunctionCall(Mask=@n12, Materials=@n16) {materialFunction=&MF_MetalsLUT}
  n18 = MaterialFunctionCall(
   AO=@n10.AO
   Distance Blend=@n10.AO
   Fresnel=@n10.AO
   Gmap=@n10.AO
   Pre Skinned Local Position=@n10.AO
   Material=MaterialFunctionCall!(In=StaticSwitchParameter(True=@n17.Result, False=@n16) {group=Skin FX, parameterName=UseMetalLUT}) {materialFunction=/Game/Weapons/WeaponSkins/MaterialFunction/MF_EmissiveFresnel}.Result
  ) {materialFunction=&BaseFN_CharacterFunction_PostFX}
  n19 = GetMaterialAttributes(None=@n18.Result) {attributeGetTypes [#1, #2]}
  n20 = Constant() {r=1}
  n21 = OneMinus(None=ScalarParameter() {group=TowerSentinelSockMasking, parameterName=HideSock})
  n22 = VertexColor()
  n23 = Constant() {r=1}
  n24 = Constant() {r=3}
  n25 = PixelNormalWS()
  n26 = Multiply(
   A=TextureSample(UVs=Fresnel(ExponentIn=@n23, BaseReflectFractionIn=@n24, Normal=@n25) {baseReflectFraction=0.5, exponent=3}) {texture=/Game/Global/Textures/Default/Masks_and_Gradients/3pxRGBImage}.RGB
   B=VectorParameter() {group=TowerSentinel_Metal, parameterName=AnodizationFilter, defaultValue=<rgba 1 0 1 0>}.RGB
  )
  n27 = Power(
   Base=DotProduct(A=AtmosphericLightVector(), B=ReflectionVectorWS())
   Exp=ScalarParameter() {defaultValue=55, group=TowerSentinel_Metal, parameterName=Anodized_pow}
  )
  n28 = Constant() {r=512}
  n29 = LinearInterpolate(A=@n26, B=Multiply(A=Constant() {r=4}, B=@n26), Alpha=Saturate(None=Multiply(A=@n27, B=@n28)))
  n30 = TextureSampleParameter2D() {
   group=TowerSentinel_Metal
   parameterName=TowerSentinelMetals
   samplerSource=SSM_Wrap_WorldGroupSettings
   samplerType=SAMPLERTYPE_Masks
   texture=&T_WhiteMask
  }
  n31 = LinearInterpolate(
   A=ScalarParameter() {group=TowerSentinel_Metal, parameterName=AnodizedEffectIntensity}
   B=ScalarParameter() {group=TowerSentinel_Metal, parameterName=AnodizedEffectIntensity2}
   Alpha=StaticComponentMaskParameter(None=@n30.RGB) {defaultR=true, group=TowerSentinel_Metal, parameterName=TowerSentinelMetals_Channel}
  )
  n32 = TextureSampleParameter2D() {
   group=TowerSentinel_Metal
   parameterName=TSEffectMasks
   samplerSource=SSM_Wrap_WorldGroupSettings
   samplerType=SAMPLERTYPE_Masks
   texture=&T_WhiteMask
  }
  n33 = Constant()
  n34 = ShadingPathSwitch(Default=Multiply(A=Multiply(A=@n29, B=@n31) {constB=0.1}, B=@n32.R), Forward=@n33, Mobile=@n33)
  n35 = SetMaterialAttributes!(
   MaterialAttributes=@n19.MaterialAttributes
   Opacity Mask=StaticSwitchParameter(True=LinearInterpolate(A=@n20, B=@n21, Alpha=@n22.B), False=Constant() {r=1}) {group=TowerSentinel, parameterName=UseSockMask}
   Base Color=Add(A=@n19.Base Color, B=StaticSwitchParameter(True=@n34, False=@n33) {group=TowerSentinel, parameterName=UseAnodizedEffect})
   Emissive Color=StaticSwitchParameter(True=Multiply(A=@n32.B, B=@n19.Emissive Color), False=@n19.Emissive Color) {parameterName=UseSockEmissiveMask}
  ) {attributeSetTypes [#3, #1, #2]}
  n36 = Constant() {r=0.001}
  n37 = VectorParameter() {group=TowerSentinel, parameterName=maskOffset, defaultValue=<rgba 0 -5 0 0>}
  n38 = Multiply(
   A=Add(A=ComponentMask(None=@n3.Result) {b=true, r=true}, B=ComponentMask(None=@n37.RGB) {g=true, r=true})
   B=ComponentMask(None=VectorParameter() {group=TowerSentinel, parameterName=maskScale, defaultValue=<rgba 2.2 1 0 0>}.RGB) {g=true, r=true}
  )
  n39 = Constant()
  n40 = OneMinus(None=Divide(A=Distance(A=@n38, B=@n39), B=ObjectRadius()))
  n41 = ScalarParameter() {defaultValue=35, group=TowerSentinel, parameterName=maskPow}
  n42 = Frac(None=Add(A=NamedRerouteUsage(), B=Constant2Vector() {r=0.5}))
  n43 = NamedRerouteUsage()
  n44 = Sine(None=ComponentMask(None=NamedRerouteUsage()) {r=true}) {period=2}
  n45 = Saturate(None=Multiply(A=Power(Base=@n44)) {constB=5})
  n46 = Multiply(
   A=Multiply(A=Power(Base=@n40, Exp=@n41), B=ScalarParameter() {defaultValue=6, parameterName=maskMul})
   B=LinearInterpolate(
    A=MaterialFunctionCall!(In=@n42) {materialFunction=&/MF_TowerSentinelChestEmissive}.Result
    B=MaterialFunctionCall(In=@n43) {materialFunction=&MF_TowerSentinelChestEmissive}.Result
    Alpha=Saturate(None=@n45)
   )
  )
  n47 = TextureSampleParameter2D() {
   group=TowerSentinel
   parameterName=TowerSentinelEmissiveMask
   samplerType=SAMPLERTYPE_Masks
   texture=/Game/Characters/Player/Female/Medium/Bodies/F_MED_Tower_Sentinel/Textures/T_F_MED_Tower_Sentinel_Body_ChestMask
  }
  StaticSwitchParameter(
   True=BlendMaterialAttributes(
    A=@n35
    B=MakeMaterialAttributes!(
     BaseColor=@n36
     Metallic
     Specular=@n36
     Roughness
     Anisotropy
     EmissiveColor=@n46
     Opacity
     OpacityMask
     Normal
     Tangent
     WorldPositionOffset
     SubsurfaceColor
     ClearCoat
     ClearCoatRoughness
     AmbientOcclusion
     Refraction
     CustomizedUVs_0
     CustomizedUVs_1
     CustomizedUVs_2
     CustomizedUVs_3
     CustomizedUVs_4
     CustomizedUVs_5
     CustomizedUVs_6
     CustomizedUVs_7
     PixelDepthOffset
     ShadingModel
     Displacement
    )
    Alpha=@n47.R
   )
   False=@n35
  ) {group=TowerSentinel, parameterName=UseChestEmissive}
 ]
}
